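; Spawn ability for the WildEstate TroopFactor boss, staging its emergence from a scare TV.
; After a random delay, it rotates the AI toward a point offset along the TV's right vector.
; If yaw matches, a montage plays, then the pawn unhides and TV emergence FX start; otherwise the TV is destroyed.
; Ability end unhides the pawn, enables alertness, restores AI behavior and targeting, and calls StimTarget on the TV.
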(event Custom|PopTV
  (Actor|DestroyActor (Variables|/WildEstate_TroopFactor/NPC/Pawn/BP_PlayerPawn_NonParticipant_Boss_WildEstate_TroopFactor.BP_PlayerPawn_NonParticipant_Boss_WildEstate_TroopFactor_C|GetScareTV (Variables|Default|GetTroopFactorPawn))))

(event Custom|ExecuteSpawnEffectBasedOnFacingRotation
  (bind _troopfactorpawn (Variables|Default|GetTroopFactorPawn))
  (bind (_returnvalue_roll _returnvalue_pitch _returnvalue_yaw) (Math|Rotator|BreakRotator (Transformation|GetActorRotation _troopfactorpawn)))
  ;[1 1 1] Check that we are facing correct direction for animation, if not just burst tv and enable boss
  (bind _returnvalue (Animation|Utilities|GetFacingRotation (Transformation|GetActorLocation _troopfactorpawn) (Variables|Default|GetFacingLocation)))
  (bind (_returnvalue_roll_2 _returnvalue_pitch_2 _returnvalue_yaw_2) (Math|Rotator|BreakRotator _returnvalue))
  (if (Math|Float|NearlyEqual(Float) _returnvalue_yaw_2 _returnvalue_yaw 0.5)
    (CallFunction|ContinuetoMontage)
    (else
      (CallFunction|PopTV)
      (Utilities|FlowControl|DelayUntilNextTick)
      (Ability|EndAbility))))

(event Ability|EventActivateAbility
  (bind _troopfactorpawn (Variables|Default|GetTroopFactorPawn))
  (bind _scaretv (Variables|/WildEstate_TroopFactor/NPC/Pawn/BP_PlayerPawn_NonParticipant_Boss_WildEstate_TroopFactor.BP_PlayerPawn_NonParticipant_Boss_WildEstate_TroopFactor_C|GetScareTV _troopfactorpawn))
  (Utilities|FlowControl|Delay (Math|Random|RandomFloatinRange 0.25 0.75))
  (CallFunction|InitAbility)
  (Utilities|IsValid _scaretv
    (:Is%20Valid
      (Utilities|FlowControl|DelayUntilNextTick)
      (bind _output_get (Variables|Default|SetFacingLocation (+ (Transformation|GetActorLocation _scaretv) (* (Transformation|GetRightVector (Variables|/Script/Engine.Actor|GetRootComponent _scaretv)) "1000.000000,1000.000000,1000.000000"))))
      (AI|Tasks|RotateAIToFace "None" 0 _output_get 0.05
        (:OnFinished
          (Ability|SetAIFocalPoint 0 (Variables|Default|GetFacingLocation))
          (Utilities|FlowControl|Delay (ScalableFloat|GetValueatLevel (Variables|Default|GetRotateToAnimDelay)))
          (CallFunction|ExecuteSpawnEffectBasedonFacingRotation))))
    (:Is%20Not%20Valid
      (Ability|EndAbility))))

(event Custom|ContinueToMontage
  (bind _montageinfo_montagetoplay (Utilities|Struct|BreakFortGameplayAbilityMontageInfo (Variables|GetMontageInfo_MontageToPlay)))
  (bind _troopfactorpawn (Variables|Default|GetTroopFactorPawn))
  (Ability|Tasks|PlayMontageAndWaitNotify "None" _montageinfo_montagetoplay
    (:then
      (Utilities|FlowControl|DelayUntilNextTick)
      (Utilities|FlowControl|Delay (ScalableFloat|GetValueatLevel (Variables|Default|GetAnimUnhideDelay)))
      (Class|BPPlayerPawnNonParticipantBossWildEstateTroopFactor|SetHidePawn _troopfactorpawn false :!class /WildEstate_TroopFactor/NPC/Pawn/BP_PlayerPawn_NonParticipant_Boss_WildEstate_TroopFactor.BP_PlayerPawn_NonParticipant_Boss_WildEstate_TroopFactor_C)
      (Class|BPTVScareTroopFactorManager|StartEmergenceFX (Variables|/WildEstate_TroopFactor/NPC/Pawn/BP_PlayerPawn_NonParticipant_Boss_WildEstate_TroopFactor.BP_PlayerPawn_NonParticipant_Boss_WildEstate_TroopFactor_C|GetScareTV _troopfactorpawn)))
    (:OnCompleted
      _)
    (:OnInterrupted
      _)
    (:OnCancelled
      _))
  (Ability|EndAbility))

(event Ability|EventOnEndAbility
  (params (bWasCancelled :type Boolean))
  (bind _troopfactorpawn (Variables|Default|GetTroopFactorPawn))
  (bind _troopfactorpawn_2 (Variables|Default|GetTroopFactorPawn))
  (bind _aibotcontroller (Variables|Default|GetAIBotController))
  (bind _scaretv (Variables|/WildEstate_TroopFactor/NPC/Pawn/BP_PlayerPawn_NonParticipant_Boss_WildEstate_TroopFactor.BP_PlayerPawn_NonParticipant_Boss_WildEstate_TroopFactor_C|GetScareTV _troopfactorpawn))
  (Utilities|IsValid _troopfactorpawn_2
    (:Is%20Valid
      ;[1 1 1] Requres TroopFactorPawn Valid
      (Class|BPPlayerPawnNonParticipantBossWildEstateTroopFactor|SetHidePawn _troopfactorpawn_2 false :!class /WildEstate_TroopFactor/NPC/Pawn/BP_PlayerPawn_NonParticipant_Boss_WildEstate_TroopFactor.BP_PlayerPawn_NonParticipant_Boss_WildEstate_TroopFactor_C)
      (Variables|/Game/Athena/AI/NPCs/Base/BP_PlayerPawn_NonParticipant.BP_PlayerPawn_NonParticipant_C|SetUseAlertness true _troopfactorpawn_2)
      (Variables|/Game/Athena/AI/NPCs/Base/BP_PlayerPawn_NonParticipant.BP_PlayerPawn_NonParticipant_C|SetUseAlertSound true _troopfactorpawn_2)
      (Utilities|IsValid _aibotcontroller
        (:Is%20Valid
          ;[1 1 1] Only necessary if TroopFactorPawn is valid and has AIBotController
          (AI|PopBehaviorControl _aibotcontroller)
          (Class|FortAthenaAIBotController|ClearAIFocalPoint _aibotcontroller false)
          (Class|FortAthenaAIBotController|ResumeTargetFocusing _aibotcontroller)
          (Utilities|IsValid _scaretv
            (:Is%20Valid
              ;[1 1 1] Requires TroopFactor, AIBotController and ScareTV for call
              (Class|BPTVScareTroopFactorManager|StimTarget _scaretv))))))))

(fn InitAbility
  (bind _output_get (Variables|Default|SetFortPlayerPawn (CallFunction|GetAvatarasFortPlayerPawn)))
  (bind _asbp_player_pawn_non_participant_boss_wild_estate_troop_factor (Utilities|Casting|CastToBP_PlayerPawn_NonParticipant_Boss_WildEstate_TroopFactor _output_get)
    (:then
      (bind _output_get_2 (Variables|Default|SetTroopFactorPawn _asbp_player_pawn_non_participant_boss_wild_estate_troop_factor))
      (bind _asfort_athena_aibot_controller (Utilities|Casting|CastToFortAthenaAIBotController (AI|GetAIController _output_get_2))
        (:then
          (Variables|Default|SetAIBotController _asfort_athena_aibot_controller))
        (:CastFailed
          _)))
    (:CastFailed
      _))
  (Ability|EndAbility))
